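; Gameplay cue for the Grav Well's fired wave, extending the parent cue's Niagara burst on the target player pawn.
; After calling the parent burst, it sets an "IsGrounded" parameter on a spawned particle component from whether the pawn's character movement is not falling.
; It also shows a gunshot-type local sound indicator at the pawn's location, passing that pawn's player controller.
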(fn OnBurst(Niagara)
  (params (MyTarget :type Actor%20Object%20Reference) (Parameters :type Gameplay%20Cue%20Parameters%20Structure%20%28by%20ref%29) (ParticleComponents :type Array%20of%20Niagara%20Particle%20System%20Component%20Object%20References) (AudioComponents :type Array%20of%20Audio%20Component%20Object%20References) (BurstCameraShakeInstance :type Camera%20Shake%20Base%20Object%20Reference) (BurstDecalInstance :type Decal%20Actor%20Object%20Reference))
  (|CallParentFunction MyTarget Parameters ParticleComponents AudioComponents BurstCameraShakeInstance BurstDecalInstance)
  (bind _asplayer_pawn_athena (Utilities|Casting|CastToPlayerPawn_Athena MyTarget))
  (Effects|Components|ParticleSystem|SetBoolParameter (Utilities|Array|Get(acopy) ParticleComponents) "IsGrounded" (not (AI|Components|NavMovement|IsFalling (Variables|/Script/Engine.Character|GetCharacterMovement _asplayer_pawn_athena))))
  (Fortnite|ShowSoundIndicatorLocal 0 _asplayer_pawn_athena (Transformation|GetActorLocation _asplayer_pawn_athena) 1e+04 "Gunshot" (Utilities|Array|MakeArray (Fortnite|GetFortPlayerControllerfromActor _asplayer_pawn_athena)) _asplayer_pawn_athena)
  (return))
