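; Gameplay cue notify for a pawn struck by a GravWell, adding audio modulation on top of the parent cue's behavior.
; On burst, it calls the parent, then activates a control bus mix if the target Fortnite player pawn is locally viewed, otherwise deactivates it.
; On remove, it deactivates the mix for a valid target before calling the parent, while execute simply defers to the parent.
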(fn OnBurst
  (params (Target :type Actor%20Object%20Reference) (Parameters :type Gameplay%20Cue%20Parameters%20Structure%20%28by%20ref%29) (SpawnResults :type Gameplay%20Cue%20Notify%20Spawn%20Result%20Structure%20%28by%20ref%29))
  (|CallParentFunction Target Parameters SpawnResults)
  (Utilities|IsValid Target
    (:Is%20Valid
      (bind _asfort_player_pawn_athena (Utilities|Casting|CastToFortPlayerPawnAthena Target))
      (if (Pawn|IsLocallyViewed _asfort_player_pawn_athena)
        (Audio|Modulation|ActivateControlBusMix Target (Variables|Default|GetContrlBusMix))
        (Audio|Modulation|DeactivateControlBusMix 0 (Variables|Default|GetContrlBusMix))
        (return)))))

(fn OnExecute
  (params (MyTarget :type Actor%20Object%20Reference) (Parameters :type Gameplay%20Cue%20Parameters%20Structure%20%28by%20ref%29))
  (return (|CallParentFunction MyTarget Parameters)))

(fn OnRemove
  (params (MyTarget :type Actor%20Object%20Reference) (Parameters :type Gameplay%20Cue%20Parameters%20Structure%20%28by%20ref%29))
  (Utilities|IsValid MyTarget
    (:Is%20Valid
      (Audio|Modulation|DeactivateControlBusMix MyTarget (Variables|Default|GetContrlBusMix)))
    (:Is%20Not%20Valid
      _))
  (return (|CallParentFunction MyTarget Parameters)))
